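# Sound cue that plays the Hunting Rifle equip foley sound.
# A single wave player feeds a modulator whose volume range is fixed at full level.
# It uses the weapon reload attenuation and weapon foley sound class, with a reduced cue volume and slightly raised pitch.
# The cue overrides concurrency and has its debug flag switched on.
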SoundCue /WeaponsUnvaulted/Assets/Audio/HuntingRifle/SC_HuntingRifle_Equip
root=Modulator_0
nodes [
 {id=Modulator_0, children [WavePlayer_0], properties {volumeMax=1, volumeMin=1}}
 {id=WavePlayer_0, properties {soundWaveAssetPtr=/WeaponsUnvaulted/Assets/Audio/HuntingRifle/SoundWaves/HuntingRifle_Equip}}
]
properties {
 attenuationSettings=&Fort_Weapon_Reload_Att
 bDebug=true
 bOverrideConcurrency=true
 maxDistance=2250
 pitchMultiplier=1.1
 soundClassObject=&Fort_WeaponFoley
 volumeMultiplier=0.5
}
