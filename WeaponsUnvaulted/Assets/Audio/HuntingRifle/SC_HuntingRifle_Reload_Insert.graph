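# Sound cue for the Hunting Rifle reload insert foley, using the weapon foley sound class and weapon reload attenuation.
# A Modulator feeds a quality-level switch that picks between two equally weighted Random nodes.
# The first two quality slots draw from four insert waves, while the last two reuse only the first two waves.
# The cue plays at reduced volume, overrides concurrency and has its debug flag on.
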SoundCue /WeaponsUnvaulted/Assets/Audio/HuntingRifle/SC_HuntingRifle_Reload_Insert
root=Modulator_0
nodes [
 {id=Modulator_0, children [QualityLevel_0]}
 {id=QualityLevel_0, children [Random_0,Random_0,Random_1,Random_1]}
 {id=Random_0, children [WavePlayer_0,WavePlayer_2,WavePlayer_3,WavePlayer_4], properties {weights [1,1,1,1]}}
 {id=Random_1, children [WavePlayer_5,WavePlayer_6], properties {weights [1,1]}}
 {id=WavePlayer_0, properties {soundWaveAssetPtr=/WeaponsUnvaulted/Assets/Audio/HuntingRifle/SoundWaves/HuntingRifle_Reload_Insert_01}}
 {id=WavePlayer_2, properties {soundWaveAssetPtr=/WeaponsUnvaulted/Assets/Audio/HuntingRifle/SoundWaves/HuntingRifle_Reload_Insert_02}}
 {id=WavePlayer_3, properties {soundWaveAssetPtr=/WeaponsUnvaulted/Assets/Audio/HuntingRifle/SoundWaves/HuntingRifle_Reload_Insert_03}}
 {id=WavePlayer_4, properties {soundWaveAssetPtr=/WeaponsUnvaulted/Assets/Audio/HuntingRifle/SoundWaves/HuntingRifle_Reload_Insert_04}}
 {id=WavePlayer_5, properties {soundWaveAssetPtr=&HuntingRifle_Reload_Insert_01}}
 {id=WavePlayer_6, properties {soundWaveAssetPtr=&HuntingRifle_Reload_Insert_02}}
]
properties {
 attenuationSettings=&Fort_Weapon_Reload_Att
 bDebug=true
 bOverrideConcurrency=true
 maxDistance=2250
 soundClassObject=&Fort_WeaponFoley
 volumeMultiplier=0.3
}
